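# Masked, lit GPU sprite material for saucer splash particles, using water splash textures with a randomly picked sub-UV cell per particle.
# Fakes reflections by sampling a cubemap along the reflection vector, desaturated, scaled by sun color and power-sharpened, then added to particle-color emissive.
# Particle alpha drives mask erosion, multiplied by near-camera fade and, except on mobile, a cutoff hiding fast particles.
# The splash normal map uses an explicit mip level and is flattened.
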Material /BRCosmetics/Effects/Fort_Effects/Materials/Pizza/M_SaucerSplash_Masked_GPU_Shaded_FakeCubemap
graph {
 properties {bUsedWithParticleSprites=true, blendMode=BLEND_Masked, pixelDepthOffsetMode=PDOM_Legacy, usedShadingModels=Default Lit}
 nodes [
  n1 = ParticleColor()
  n2 = Add(A=@n1.RGB, B=Multiply())
  n3 = ScalarParameter() {defaultValue=0.1, parameterName=Emissive}
  n4 = ComponentMask(None=CollectionParameter() {collection=&FortniteMaterialParameters, parameterName=SunLightColor}) {b=true, g=true, r=true}
  n5 = TextureSample(UVs=ReflectionVectorWS()) {samplerType=SAMPLERTYPE_LinearColor, texture=/Game/Weapons/WeaponSkins/Textures/AthenaCubemap2}
  n6 = Constant() {r=1}
  n7 = Power(
   Base=Multiply(A=Clamp(None=@n4) {maxDefault=2, minDefault=0.5}, B=Desaturation(None=@n5.RGB, Fraction=@n6))
   Exp=ScalarParameter() {defaultValue=6, parameterName=CubemapPower}
  )
  n8 = ScalarParameter() {defaultValue=1, parameterName=Cubemap Scale}
  MaterialFunctionCall(Emissive=Add(A=Multiply(A=@n2, B=@n3), B=Multiply(A=@n7, B=@n8))) {materialFunction=&MF_FX_EmissiveAdjustments}
  n9 = ParticleColor()
  n10 = MaterialFunctionCall!(
   UseTexCoord 1
   Rows=ScalarParameter() {defaultValue=2, parameterName="Rows (S)"}
   Columns=ScalarParameter() {defaultValue=2, parameterName="Columns (S)"}
   Particle Random Value=ParticleRandom()
  ) {materialFunction=/Game/Effects/Fort_Effects/Materials/_Functions/MF_RandomSubUV_GPU_Sprites}
  n11 = TextureSample(UVs=@n10.Result) {samplerType=SAMPLERTYPE_LinearColor, texture=/Game/Effects/Fort_Effects/Textures/Water/T_waterSplash}
  n12 = LinearInterpolate(A=@n11.R, B=@n11.R, Alpha=@n9.A)
  n13 = MaterialFunctionCall(Distance=ScalarParameter() {defaultValue=125, parameterName="Distance (S)"}) {materialFunction=&Near_Camera_Fade}
  n14 = Constant() {r=1}
  n15 = Constant()
  n16 = Clamp(None=Subtract(A=ParticleSpeed(), B=ScalarParameter() {defaultValue=500, parameterName=ParticleSpeed_FadeAboveThisValue}))
  Multiply(
   A=Multiply(
    A=MaterialFunctionCall!(Dynamic Param=@n9.A, Alpha=@n12) {materialFunction=/Game/Effects/Fort_Effects/Materials/_Functions/Erode_Alpha}.Result
    B=Clamp(None=@n13.Result)
   )
   B=ShadingPathSwitch(Default=LinearInterpolate(A=@n14, B=@n15, Alpha=@n16), Mobile=Constant() {r=1}) {desc=Migrated from FeatureLevelSwitch}
  )
  n17 = TextureSample!(UVs=@n10.Result, Tex, Level=ScalarParameter() {defaultValue=3, parameterName=NormalMipLevel}, Apply View MipBias) {mipValueMode=TMVM_MipLevel, samplerType=SAMPLERTYPE_Normal, texture=/Game/Effects/Fort_Effects/Textures/Water/T_waterSplash_N}
  n18 = ScalarParameter() {defaultValue=0.5, parameterName="Flatness (S)"}
  Reroute(None=Normalize(VectorInput=MaterialFunctionCall(Normal=@n17.RGB, Flatness=@n18) {materialFunction=&FlattenNormal}.Result))
  Reroute(None=@n1.RGB)
  ScalarParameter() {defaultValue=1, parameterName=Specular}
  ScalarParameter() {defaultValue=1, parameterName=Metallic}
  ScalarParameter() {defaultValue=1, parameterName=Roughness}
 ]
}
