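# M_ScreenSplat is a translucent particle-sprite material for a liquid splat effect, likely on screen, in the Pizza effects set.
# Particle color multiplies the red channel of a packed 6x6 splash-shape flipbook texture.
# For alpha, cloud noise raised to a dynamic Exp parameter has the inverted splash alpha subtracted, then scales by particle alpha, giving adjustable noisy breakup.
# An Erode_Alpha function node is present with unwired inputs.
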Material /BRCosmetics/Effects/Fort_Effects/Materials/Pizza/M_ScreenSplat
graph {
 properties {bUsedWithParticleSprites=true, blendMode=BLEND_Translucent, pixelDepthOffsetMode=PDOM_Legacy, usedShadingModels=Default Lit}
 nodes [
  MaterialFunctionCall!(Dynamic Param, Alpha) {materialFunction=/Game/Effects/Fort_Effects/Materials/_Functions/Erode_Alpha}
  n1 = ParticleColor()
  n2 = TextureSample() {texture=/Game/Effects/Fort_Effects/Textures/Liquid/T_PoisonSplashShapes_6X6_Packed}
  Multiply(A=@n1.RGB, B=@n2.R)
  n3 = TextureSample() {texture=&AtmosphericCloudNoise01}
  n4 = DynamicParameter() {defaultValue=<rgba 0 1 1 1>, paramNames [Exp,Param2,Param3,Param4]}
  Multiply(A=@n1.A, B=Subtract(A=Power(Base=@n3.G, Exp=@n4.Exp), B=OneMinus(None=@n2.A)))
 ]
}
